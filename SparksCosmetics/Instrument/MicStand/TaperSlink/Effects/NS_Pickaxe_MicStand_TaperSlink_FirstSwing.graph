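# Niagara effect for the first swing of the TaperSlink mic stand pickaxe, built from three short-lived burst emitters.
# AirflowColor, slightly delayed, places a fading mesh on the user-supplied skeletal mesh, stretching it and animating material parameters.
# Sparks, also delayed, bursts dragged, shrinking sprites from that mesh in randomized colors, with a second sprite renderer and glow parameters.
# NE_SoundWave spawns a stack of growing rings-like sprites, staggering offset, size and fade by spawn index.
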NiagaraSystem /SparksCosmetics/Instrument/MicStand/TaperSlink/Effects/NS_Pickaxe_MicStand_TaperSlink_FirstSwing
dependencies {
 usedDataInterfaces [
  &NiagaraDataInterfaceCurve
  &NiagaraDataInterfaceArrayColor
  &NiagaraDataInterfaceColorCurve
  &NiagaraDataInterfaceVector2DCurve
 ]
 usedDynamicInputs []
 usedModules [&SystemState]
}
emitters {
 AirflowColor {
  emitterUpdateScript [
   &EmitterState(Life Cycle Mode=::NewEnumerator1, Loop Behavior=::NewEnumerator1, Loop Delay=0.07, Loop Duration=0.5, UseLoopDelay=true)
   &SpawnBurst_Instantaneous(Use Spawn Probability=false, UseLoopCountLimit=false)
  ]
  particleSpawnScript [
   &InitializeParticle(Lifetime=0.25, Mesh Scale=<0.8 0.8 0.8>, Mesh Scale Mode=::NewEnumerator4, UsePositionOffset=false, Write Material Random=true)
   &InitialMeshOrientation(Mesh Orientation Mode=::NewEnumerator2, Rotation=<0 -0.25 0>, Use Rotation Vector=true)
   &SkeletalMeshLocation(
    MaskSpawnGroup=false
    Mesh Sampling Type=::NewEnumerator5
    Orientation Sampling=::NewEnumerator1
    Sampled Position Offset=<0 0 50>
    Skeletal Mesh=$User.Skeletal Mesh
    Socket Sampling Mode=::NewEnumerator5
   )
   &CameraOffset(Camera Offset Amount=30)
  ]
  particleUpdateScript [
   &ParticleState()
   &ScaleMeshSize(Scale Factor=&MakeVector(
     X=&FloatFromCurve(FloatCurve=DataInterface() {propertyValues {Curve {keys [<at 0 1>,<at 1 2>]}}})
     Y=1
     Z=1
    ))
   &SolveForcesAndVelocity(Clamp Velocity=false, Limit Acceleration=false)
   &ScaleColor(
    Scale Alpha=&FloatFromCurve(FloatCurve=DataInterface() {propertyValues {Curve {keys [<at 0 1 RCIM_Cubic>,<at 1 0>]}}})
    ScaleA=true
    ScaleRGB=true
   )
   &DynamicMaterialParameters(
    Index 0 Param 1=&FloatFromCurve(FloatCurve=DataInterface() {propertyValues {Curve {keys [<at 0 0.4>,<at 1 0.8>]}}})
    Index 0 Param 2=&FloatFromCurve(FloatCurve=DataInterface() {propertyValues {Curve {keys [
         {arriveTangent=-2.2616, interpMode=RCIM_Cubic, leaveTangent=-3.3361, tangentMode=RCTM_Break, time=0, value=1}
         {arriveTangent=-0.8527, interpMode=RCIM_Cubic, leaveTangent=-0.8527, tangentMode=RCTM_User, time=0.8, value=0}
        ]}}})
    Index 0 Param 3=0.4
    Module.Param0WriteEnabled=true
    Module.Param1WriteEnabled=true
    Module.Param2WriteEnabled=true
    Module.Param3WriteEnabled=true
   ) {module=DynamicMaterialParameters002}
  ]
  rendererClasses [&NiagaraMeshRendererProperties]
  renderers [{rendererClass=&NiagaraMeshRendererProperties, rendererIndex=0}]
 }
 NE_SoundWave {
  emitterUpdateScript [
   &EmitterState()
   &SpawnBurst_Instantaneous(Spawn Count=6, Use Spawn Probability=false, UseLoopCountLimit=false)
  ]
  particleSpawnScript [
   &InitializeParticle(
    Color Mode=::NewEnumerator0
    Lifetime=0.35
    Position Offset=&Add_Vector(
     A=&Multiply_VectorByFloat(Vector=<0 0 50>, Float=&ReturnNormalizedExecIndex())
     B=<0 0 -80>
    )
    Sprite Rotation Mode=::NewEnumerator2
    Sprite Size Mode=::NewEnumerator3
    Uniform Sprite Size=&FloatFromCurve(
     FloatCurve=DataInterface() {propertyValues {Curve {
        defaultValue=0.25
        keys [
         {arriveTangent=2.8511, interpMode=RCIM_Cubic, leaveTangent=2.8511, tangentMode=RCTM_User, time=0, value=0.5}
         {arriveTangent=0.5, interpMode=RCIM_Cubic, leaveTangent=0.5, time=0.2262, value=0.8779}
         {tangentMode=RCTM_User, time=1, value=1}
        ]
       }}}
     CurveIndex=&ReturnNormalizedExecIndex()
     Scale Curve=80
    )
    UsePositionOffset=true
    Write Material Random=true
   )
   &CameraOffset(Camera Offset Amount=10)
   &AddVelocity(Velocity Speed Scale=&Lerp_Float(A=0, B=-1, Alpha=&ReturnNormalizedExecIndex()))
   &SpriteFacingAndAlignment(Sprite Facing=<0 0 1>, WriteAlignment=false, WriteFacing=true)
  ]
  particleUpdateScript [
   &ParticleState()
   &ScaleSpriteSize(
    Uniform Curve Index=&NormalizeFloat(
     Value=&Multiply_Float(
      A=$Particles.NormalizedAge
      B=&FloatFromCurve(
       FloatCurve=DataInterface() {propertyValues {Curve {defaultValue=0.5, keys [{tangentMode=RCTM_User, time=0, value=0.5}, {tangentMode=RCTM_User, time=1, value=1}]}}}
       CurveIndex=&ReturnNormalizedExecIndex()
      )
     )
     Min=0
     Max=1
    )
    Uniform Curve Sprite Scale=DataInterface() {propertyValues {Curve {keys [
        {arriveTangent=2.0053, interpMode=RCIM_Cubic, leaveTangent=2.0053, tangentMode=RCTM_User, time=0, value=0}
        {arriveTangent=0.1491, interpMode=RCIM_Cubic, leaveTangent=0.1491, tangentMode=RCTM_User, time=1, value=1}
       ]}}}
   )
   &ScaleColor(
    Scale Alpha=&FloatFromCurve(
     FloatCurve=DataInterface() {propertyValues {Curve {
        defaultValue=0
        keys [
         {arriveTangent=6.6118, interpMode=RCIM_Cubic, leaveTangent=6.6118, tangentMode=RCTM_User, time=0, value=0}
         {interpMode=RCIM_Cubic, tangentMode=RCTM_User, time=0.5, value=1}
         {tangentMode=RCTM_User, time=1, value=0}
        ]
       }}}
     CurveIndex=&NormalizeFloat(
      Value=&Multiply_Float(
       A=$Particles.NormalizedAge
       B=&FloatFromCurve(
        FloatCurve=DataInterface() {propertyValues {Curve {defaultValue=0.5, keys [{tangentMode=RCTM_User, time=0, value=0.5}, {tangentMode=RCTM_User, time=1, value=1.5}]}}}
        CurveIndex=&ReturnNormalizedExecIndex()
       )
      )
      Min=0
      Max=1
     )
     Scale Curve=1.5
    )
    Scale RGB=&MakeVectorFromLinearColorRGB(Color=&ColorFromCurve(ColorCurve=DataInterface() {propertyValues {
        AlphaCurve {keys [<at 0 1 RCIM_Cubic>]}
        BlueCurve {keys [
          <at 0 0.0144 RCIM_Cubic>
          <at 0.3932 0.0063>
          {arriveTangent=-0.0104, interpMode=RCIM_Cubic, leaveTangent=-0.0104, time=0.8407, value=0.0221}
          <at 1 0>
         ]}
        GreenCurve {keys [
          <at 0 0.8 RCIM_Cubic>
          <at 0.3932 0.6986>
          {arriveTangent=-0.2758, interpMode=RCIM_Cubic, leaveTangent=-0.2758, time=0.8407, value=0.5312}
          <at 1 0.5312>
         ]}
        RedCurve {keys [
          <at 0 0.5428 RCIM_Cubic>
          <at 0.3932 0.3651>
          {arriveTangent=-0.5288, interpMode=RCIM_Cubic, leaveTangent=-0.5288, time=0.8407, value=0.1797}
          <at 1 0.0443>
         ]}
       }}))
    ScaleA=true
    ScaleRGB=true
   )
   &SolveForcesAndVelocity(Clamp Velocity=false, Limit Acceleration=false)
  ]
  rendererClasses [&NiagaraSpriteRendererProperties]
  renderers [{rendererClass=&NiagaraSpriteRendererProperties}]
 }
 Sparks {
  emitterUpdateScript [
   &EmitterState(Life Cycle Mode=::NewEnumerator1, Loop Behavior=::NewEnumerator1, Loop Delay=0.05, Loop Duration=0.5, UseLoopDelay=true)
   &SpawnBurst_Instantaneous(Spawn Count=10, Use Spawn Probability=false, UseLoopCountLimit=false)
   &SpawnPerUnit(Module.Use Max Movement Threshold=true, Module.Use Movement Tolerance=false, Use Spawn Probability=false) {enabled=false}
  ]
  particleSpawnScript [
   &PartitionParticles(Partition Mode:ENiagara_PartitionMode=::NewEnumerator5)
   &InitializeParticle(
    Color=&MultiplyLinearColor(
     Linear Color=&SelectLinearColorFromArray(
      Array Sampling Mode=::NewEnumerator0
      Direct Array Index=?
      Color Selection Array=DataInterface() {propertyValues {ColorData [<rgba 0.3096 1 0 1>, <rgba 0.4452 0 1 1>]}}
      Random Seed=0
     )
     Scale Factor=&MakeVector4_FromLinearColor(LinearColor=&ColorFromCurve(
       ColorCurve=DataInterface() {propertyValues {
         AlphaCurve {keys [<at 0 1>,<at 1 1>]}
         BlueCurve {keys [<at 0.0013 1>,<at 0.119 1>,<at 0.1906 0.5>,<at 0.3109 1>,<at 0.3978 1>,<at 0.5 0.5>,<at 0.6667 0.5>,<at 0.8333 0.5>,<at 1 0.5>]}
         GreenCurve {keys [<at 0.0013 1>,<at 0.119 1>,<at 0.1906 0.5058>,<at 0.3109 0.5>,<at 0.3978 0.5715>,<at 0.5 1>,<at 0.6667 0.9189>,<at 0.8333 0.6329>,<at 1 0.5058>]}
         RedCurve {keys [<at 0.0013 1>,<at 0.119 1>,<at 0.1906 1>,<at 0.3109 0.7784>,<at 0.3978 0.5>,<at 0.5 0.5678>,<at 0.6667 1>,<at 0.8333 1>,<at 1 1>]}
        }}
       CurveIndex=&RandomRangeFloat(Minimum=0, Maximum=1)
      ))
    )
    Lifetime Max=0.4
    Lifetime Min=0.3
    Lifetime Mode=::NewEnumerator1
    Sprite Size Max=<1 2>
    Sprite Size Min=<1 2>
    Sprite Size Mode=::NewEnumerator2
    UsePositionOffset=false
    Write Material Random=true
   )
   &SkeletalMeshLocation(Bone Sampling Mode=::NewEnumerator4, MaskSpawnGroup=false, Sampled Position Offset=<0 -3 20>, Skeletal Mesh=$User.Skeletal Mesh)
   &AddVelocity(Velocity=<0 0 -500>)
   &BoxLocation(Box Size=<10 10 100>, MaskSpawnGroup=false, Module.Override Seed=false, Rotate Box=false, Surface Only=false)
  ]
  particleUpdateScript [
   &ParticleState()
   &Color(Color=&ColorFromCurve(ColorCurve=DataInterface() {propertyValues {
       AlphaCurve {keys [<at 0 1>,<at 1 1>]}
       BlueCurve {keys [<at 0 0.5>,<at 0.1673 1>,<at 0.3333 1>,<at 0.5 0.5>,<at 0.6667 0.5>,<at 0.8333 0.5>,<at 1 0.5>]}
       GreenCurve {keys [<at 0 0.5058>,<at 0.1673 0.5>,<at 0.3333 0.5715>,<at 0.5 1>,<at 0.6667 0.9189>,<at 0.8333 0.6329>,<at 1 0.5058>]}
       RedCurve {keys [<at 0 1>,<at 0.1673 0.7784>,<at 0.3333 0.5>,<at 0.5 0.5678>,<at 0.6667 1>,<at 0.8333 1>,<at 1 1>]}
      }})) {enabled=false}
   /SetVariables(Particles.Glow03_Color=&MakeLinearColorFromVectorAndFloat(
     "Vector (RGB)"=&MakeVectorFromLinearColorRGB(Color=$Particles.Initial.Color)
     "Float (Alpha)"=&Multiply_Float(A=0.2, B=1)
    )) {isSetParameters=true}
   /SetVariables(Particles.Glow03_Size=&Multiply_Vector2D_ByFloat(
     Float=&FloatFromCurve(
      FloatCurve=DataInterface() {propertyValues {Curve {keys [
          <at 0 0 RCIM_Cubic>
          <at 0.2 1.3 RCIM_Cubic>
          {arriveTangent=-0.1423, interpMode=RCIM_Cubic, leaveTangent=-0.1423, tangentMode=RCTM_User, time=1, value=0}
         ]}}}
      Scale Curve=10
     )
     Vector2D=<0.7 2>
    )) {isSetParameters=true}
   &ScaleColor(
    Scale Alpha=&FloatFromCurve(FloatCurve=DataInterface() {propertyValues {Curve {keys [<at 0.2 0>,<at 0.264 0>,<at 0.3 1>,<at 1 1>]}}})
    ScaleA=true
    ScaleRGB=true
   )
   &ScaleSpriteSize(Scale Factor=&Vector2DFromCurve(Vector2Curve=DataInterface() {propertyValues {XCurve {keys [<at 0 1>,<at 1 0>]}, YCurve {keys [<at 0 1>,<at 1 0>]}}}, ScaleCurve=<0.3 1>)) {module=ScaleSpriteSize001}
   &CurlNoiseForce(
    Noise Frequency=300
    Noise Strength=&Multiply_Float(A=2000, B=1)
    Pan Noise=true
    Pan Noise Field=<0 0 1>
    Random Seed=10
   ) {enabled=false}
   &Drag(
    Drag=&FloatFromCurve(FloatCurve=DataInterface() {propertyValues {Curve {keys [
         {interpMode=RCIM_Cubic, leaveTangent=12.2797, tangentMode=RCTM_Break, time=0, value=1}
         {interpMode=RCIM_Cubic, tangentMode=RCTM_User, time=1, value=5}
        ]}}})
    Use Linear Drag=true
    Use Rotational Drag=false
   )
   &SolveForcesAndVelocity(Clamp Velocity=false, Limit Acceleration=false)
   &ScaleSpriteSizeBySpeed(Max Scale Factor=<1 2>)
  ]
  rendererClasses [&NiagaraSpriteRendererProperties]
  renderers [
   {rendererClass=&NiagaraSpriteRendererProperties}
   {rendererClass=&NiagaraSpriteRendererProperties, rendererIndex=1}
  ]
 }
}
summary {
 systemName=NS_Pickaxe_MicStand_TaperSlink_FirstSwing
 emitters [
  {emitterName=AirflowColor, rendererClasses [&NiagaraMeshRendererProperties]}
  {emitterName=Sparks, rendererClasses [&NiagaraSpriteRendererProperties]}
  {emitterName=NE_SoundWave, rendererClasses [&NiagaraSpriteRendererProperties]}
 ]
 userVariables [User.Skeletal Mesh {
   type=NiagaraDataInterfaceSkeletalMesh
   defaultValue {
    struct=NiagaraExt_VariableValue_DataInterface
    value {dataInterface=&self:NiagaraDataInterfaceSkeletalMesh_0}
   }
  }]
}
