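# Niagara impact effect for the TaperSlink mic stand pickaxe, layering a colored flash burst, surface debris, lingering smoke and falling sparks.
# The core burst uses spawn groups and masks to drive several sprite layers and a light from one emitter.
# Debris meshes and motion follow the User.PhysicalSurface input, and the upward smoke poof emitter is disabled.
# System-level scale, direction and time-of-day tint feed the emitters, with separate front-end preview tuning.
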NiagaraSystem /SparksCosmetics/Instrument/MicStand/TaperSlink/Effects/NS_Pickaxe_MicStand_TaperSlink_Impact
dependencies {
 usedDataInterfaces [
  &NiagaraDataInterfaceCurve
  &NiagaraDataInterfaceArrayColor
  &NiagaraDataInterfaceColorCurve
  &NiagaraDataInterfaceVectorCurve
  &NiagaraDataInterfaceSpriteRendererInfo
  &NiagaraDataInterfaceArrayFloat
 ]
 usedDynamicInputs []
 usedModules [
  &self:SystemSpawnScript.SystemScriptSource.SystemScriptGraph.NiagaraNodeAssignment_0.SetVariables_#1
  &SystemState
 ]
}
emitters {
 Core_Impact_Burst001 {
  emitterUpdateScript [
   &EmitterState()
   &SpawnBurst_Instantaneous(
    Spawn Probability=&FN_FrontEndFloatModifier(Gameplay=1, Front End=1)
    Use Spawn Probability=true
    UseLoopCountLimit=false
   ) {module=SpawnBurst_Instantaneous001}
   &SpawnBurst_Instantaneous(Spawn Group=1, Use Spawn Probability=false, UseLoopCountLimit=false)
   &SpawnBurst_Instantaneous(Spawn Group=2, Use Spawn Probability=false, UseLoopCountLimit=false) {module=SpawnBurst_Instantaneous002}
   &SpawnBurst_Instantaneous(
    Spawn Count=10
    Spawn Group=3
    Spawn Probability=&FN_FrontEndFloatModifier(Gameplay=1, Front End=1)
    Use Spawn Probability=true
   ) {module=SpawnBurst_Instantaneous003}
   &SpawnBurst_Instantaneous(Spawn Count=10, Spawn Time=0.0667, Use Spawn Probability=false, UseLoopCountLimit=false) {module=SpawnBurst_Instantaneous004}
  ]
  particleSpawnScript [
   &PartitionParticles(Number Of Partitions=3)
   /SetVariables(Particles.VisibilityTag=&SpawnGroup()) {isSetParameters=true}
   &InitializeParticle(
    Lifetime=&MaskFloatBySpawnGroup(
     Masked Float=&FN_FrontEndFloatModifier(Gameplay=0.22, Front End=0.18)
     Passthrough Float=&MaskFloatBySpawnGroup(
      Masked Float=0.35
      Passthrough Float=&MaskFloat(
       Mask Source=0
       Mask Parameter=?
       Mask=3
       Masked Float=&RandomRangeFloat(Minimum=0.18, Maximum=0.2)
       Passthrough Float=1
      )
      Spawn Group Mask=2
     )
     Spawn Group Mask=1
    )
    Mesh Scale Mode=::NewEnumerator3
    Mesh Uniform Scale=0.35
    Sprite Rotation Mode=::NewEnumerator2
    Sprite Size=&MaskVector2dBySpawnGroup(
     Masked Vector2D=&UniformRangedVector2D(Minimum=<8 35>, Maximum=<11 55>)
     Passthrough Vector2D=&MaskVector2dBySpawnGroup(
      Masked Vector2D=&UniformRangedVector2D(Minimum=<5 5>, Maximum=<15 5>)
      Passthrough Vector2D=&MaskVector2dBySpawnGroup(
       Masked Vector2D=&Vector2DFromFloat(Value=&Multiply_Float(A=175, B=$System.Global Scale))
       Passthrough Vector2D=&Vector2DFromFloat(Value=&Multiply_Float(A=130, B=$System.Global Scale))
       Spawn Group Mask=2
      )
      Spawn Group Mask=4
     )
     Spawn Group Mask=3
    )
    Sprite Size Mode=::NewEnumerator4
    UsePositionOffset=false
    Write Material Random=true
   )
   &FN_PerParticleLOD(Enable SpriteSize Distance Scaling=true, Sprite Size Scale Far=<2.5 2.5>)
   &ShapeLocation(
    MaskSpawnGroup=false
    Rotation Mode=::NewEnumerator2
    Sphere Distribution=::NewEnumerator3
    Sphere Radius=&SelectFloatFromArray(
     Direct Array Index=&SpawnGroup()
     Float Selection Array=DataInterface() {propertyValues {FloatData [0,0,0,0.5,0]}}
    )
    UseOwnerScale=true
    Yaw / Pitch / Roll=&RandomRangeVector(Minimum=<0 0 0>, Maximum=<1 1 1>)
   )
   &AddVelocityFromPoint(
    Module.Use Falloff Distance=false
    Velocity Strength=&MaskFloatBySpawnGroup(
     Masked Float=&FloatFromCurve(
      FloatCurve=DataInterface() {propertyValues {Curve {keys [<at 0.0101 172.8235>,<at 0.1521 116.2374>,<at 0.375 80.9199>,<at 1 20>]}}}
      CurveIndex=&RandomRangeFloat(Minimum=0, Maximum=1)
      Scale Curve=$System.Global Scale
     )
     Passthrough Float=&MaskFloatBySpawnGroup(
      Masked Float=&FloatFromCurve(
       FloatCurve=DataInterface() {propertyValues {Curve {keys [<at 0 0>,<at 1 15>]}}}
       CurveIndex=&RandomRangeFloat(Minimum=0, Maximum=1)
       Scale Curve=$System.Global Scale
      )
      Passthrough Float=0
      Spawn Group Mask=4
     )
     Spawn Group Mask=3
    )
   )
   &InitialMeshOrientation(Orientation Axis=<0 0 1>, Orientation Vector=<0 0 1>, Rotation=<0 0.25 0>, Use Orientation Vector=true, Use Rotation Vector=true)
   &InitialMeshOrientation(
    Orientation Axis=<0 0 0>
    Orientation Vector=<0 0 0>
    Rotation=&RandomRangeVector(Minimum=<0 0 0>, Maximum=<1 0 0>)
    Use Orientation Vector=false
    Use Rotation Vector=true
   ) {module=InitialMeshOrientation001}
   /SetVariables(
    Particles.Flash Star Cam Offset=15
    Particles.Flash Star Color:LinearColor=<rgba 3 1.5 0 1>
    Particles.Flash Star Lifetime=0.18
    Particles.Flash Star Rotation:NiagaraFloat=&RandomRangeFloat(Minimum=0, Maximum=360)
    Particles.Flash Star Sprite Size:Vector2f=&Multiply_Vector2D_ByFloat(Float=&Multiply_Float(A=$System.Global Scale, B=75), Vector2D=<1 1>)
    Particles.Flash Star SubUV Index:NiagaraFloat=&RandomRangeFloat(Minimum=0, Maximum=3)
   ) {isSetParameters=true}
   /SetVariables(
    Particles.FlareCardColor:LinearColor=$System.TOD
    Particles.FlareCardDynamicParameter:Vector4f=&MakeVector4(X=0.2, Y=1, Z=0, W=0)
   ) {isSetParameters=true}
   &CameraOffset(Camera Offset Amount=&MaskFloatBySpawnGroup(Masked Float=0, Passthrough Float=20, Spawn Group Mask=3))
   &ApplyInitialForces()
   /SetVariables(Particles.DarkGlowDyn0:Vector4f=&MakeVector4(X=1, Y=1, Z=1, W=1)) {isSetParameters=true}
  ]
  particleUpdateScript [
   &ParticleState()
   &ScaleSpriteSize(
    Scale Factor=&MaskVector2dBySpawnGroup(
     Masked Vector2D=&Multiply_Vector2D_ByFloat(
      Float=&FloatFromCurve(FloatCurve=DataInterface() {propertyValues {Curve {
          defaultValue=1
          keys [
           {interpMode=RCIM_Cubic, tangentMode=RCTM_User, time=0, value=1}
           {arriveTangent=-3.0339, interpMode=RCIM_Cubic, leaveTangent=-3.0339, tangentMode=RCTM_User, time=1, value=0}
          ]
         }}})
      Vector2D=<1 1>
     )
     Passthrough Vector2D=&MaskVector2dBySpawnGroup(
      Masked Vector2D=&MakeVector2D(
       X=&FloatFromCurve(
        FloatCurve=DataInterface() {propertyValues {Curve {
           defaultValue=1
           keys [
            {arriveTangent=6.211, interpMode=RCIM_Cubic, leaveTangent=6.211, tangentMode=RCTM_User, time=0, value=0.5}
            {arriveTangentWeight=0.1005, interpMode=RCIM_Cubic, leaveTangentWeight=0.2345, tangentMode=RCTM_User, time=0.2, value=1}
            {arriveTangentWeight=0.6964, interpMode=RCIM_Cubic, tangentMode=RCTM_User, time=1.0035, value=0.3}
           ]
          }}}
        Scale Curve=$System.Global Scale
       )
       Y=&FloatFromCurve(
        FloatCurve=DataInterface() {propertyValues {Curve {
           defaultValue=1
           keys [
            {interpMode=RCIM_Cubic, leaveTangentWeight=0.34, tangentMode=RCTM_User, time=0, value=0.8}
            {arriveTangent=0.6525, interpMode=RCIM_Cubic, leaveTangent=0.6525, tangentMode=RCTM_User, time=1, value=1.2}
           ]
          }}}
        Scale Curve=$System.Global Scale
       )
      )
      Passthrough Vector2D=&MaskVector2dBySpawnGroup(
       Masked Vector2D=&MakeVector2D(
        X=&FloatFromCurve(
         FloatCurve=DataInterface() {propertyValues {Curve {
            defaultValue=1
            keys [
             {arriveTangent=1.7844, interpMode=RCIM_Cubic, leaveTangent=1.7844, tangentMode=RCTM_User, time=0, value=0}
             {arriveTangentWeight=0.3771, interpMode=RCIM_Cubic, leaveTangentWeight=0.0943, tangentMode=RCTM_User, time=0.8, value=1}
             {arriveTangent=-5.8206, interpMode=RCIM_Cubic, leaveTangent=-5.8206, tangentMode=RCTM_User, time=1.0063, value=0.1176}
            ]
           }}}
         Scale Curve=$System.Global Scale
        )
        Y=&FloatFromCurve(
         FloatCurve=DataInterface() {propertyValues {Curve {
            defaultValue=1
            keys [
             {arriveTangent=4.2951, interpMode=RCIM_Cubic, leaveTangent=4.2951, tangentMode=RCTM_User, time=0, value=0.4}
             {arriveTangent=2.8775, interpMode=RCIM_Cubic, leaveTangent=2.8775, tangentMode=RCTM_User, time=1, value=1.85}
            ]
           }}}
         Scale Curve=$System.Global Scale
        )
       )
       Passthrough Vector2D=&Vector2DFromFloat(Value=&FloatFromCurve(FloatCurve=DataInterface() {propertyValues {Curve {
            defaultValue=1
            keys [
             {arriveTangent=1.3623, interpMode=RCIM_Cubic, leaveTangent=1.3623, tangentMode=RCTM_User, time=0, value=0.5}
             {arriveTangent=0.2407, arriveTangentWeight=0.7893, interpMode=RCIM_Cubic, leaveTangent=0.2407, tangentMode=RCTM_User, time=0.993, value=1}
            ]
           }}}))
       Spawn Group Mask=4
      )
      Spawn Group Mask=3
     )
     Spawn Group Mask=1
    )
    Scale Sprite Size Mode=::NewEnumerator2
   )
   &LinearForce(Force=&Multiply_VectorByFloat(
     Vector=$System.DirectionVector
     Float=&MaskFloatBySpawnGroup(
      Masked Float=&SelectFloatFromArray(
       Direct Array Index=$Particles.PartitionParticles.Partition
       Float Selection Array=DataInterface() {propertyValues {FloatData [25,3900,4500]}}
      )
      Passthrough Float=0
      Spawn Group Mask=4
     )
    ))
   &ScaleMeshSize(Scale Factor=&VectorFromFloat(Value=&FloatFromCurve(
      FloatCurve=DataInterface() {propertyValues {Curve {keys [
          {arriveTangent=7.2985, interpMode=RCIM_Cubic, leaveTangent=7.2985, tangentMode=RCTM_User, time=0, value=0.5}
          {arriveTangent=0.6472, interpMode=RCIM_Cubic, leaveTangent=0.6472, tangentMode=RCTM_User, time=0.15, value=1.25}
          {arriveTangent=1.8898, interpMode=RCIM_Cubic, leaveTangent=1.8898, tangentMode=RCTM_User, time=1, value=2}
         ]}}}
      Scale Curve=$System.Global Scale
     )))
   &ScaleColor(
    Scale Alpha=&FloatFromCurve(FloatCurve=DataInterface() {propertyValues {Curve {
        defaultValue=1
        keys [
         {arriveTangent=-0.2961, interpMode=RCIM_Cubic, leaveTangent=-0.2961, tangentMode=RCTM_User, time=0.1, value=1}
         {arriveTangentWeight=0.4395, interpMode=RCIM_Cubic, tangentMode=RCTM_User, time=1, value=0}
        ]
       }}})
    Scale RGB=&MakeVectorFromLinearColorRGB(Color=&Multiply_LinearColorByFloat(
      Linear Color=&MultiplyLinearColor(
       Linear Color=&ColorFromCurve(ColorCurve=DataInterface() {propertyValues {
          AlphaCurve {keys [<at 0.0782 1>,<at 1 0>]}
          BlueCurve {keys [<at 0 5>,<at 0.7391 1>]}
          GreenCurve {keys [<at 0 0.001>,<at 0.7391 0.2>]}
          RedCurve {keys [<at 0 0.3487>,<at 0.7391 0.289>]}
         }})
       Scale Factor=&MakeVector4_FromLinearColor(LinearColor=$System.TOD)
      )
      Float=&FN_FrontEndFloatModifier(Gameplay=1, Front End=1)
     ))
    ScaleA=true
    ScaleRGB=true
   )
   /SetVariables(Particles.Update Flash Star Age:NiagaraFloat=&Divide_Float(A=$Particles.Age, B=$Particles.Initial.Flash Star Lifetime)) {isSetParameters=true}
   /SetVariables(
    Particles.Update Flash Star Sprize Size:Vector2f=&Multiply_Vector2D_ByFloat(
     Float=&FloatFromCurve(
      FloatCurve=DataInterface() {propertyValues {Curve {
         defaultValue=0.2
         keys [
          {arriveTangent=1.826, interpMode=RCIM_Cubic, leaveTangent=1.826, leaveTangentWeight=1.0797, tangentMode=RCTM_User, time=0.0036, value=0.2824}
          {arriveTangentWeight=0.6623, interpMode=RCIM_Cubic, leaveTangentWeight=0.3962, tangentMode=RCTM_User, time=0.627, value=0.65}
          {arriveTangent=6.378, arriveTangentWeight=0.4894, interpMode=RCIM_Cubic, leaveTangent=6.378, tangentMode=RCTM_User, time=1, value=1.5}
         ]
        }}}
      CurveIndex=$Particles.Update Flash Star Age
      Scale Curve=$System.Global Scale
     )
     Vector2D=$Particles.Initial.Flash Star Sprite Size
    )
    Particles.UpdateFlareCardColor:LinearColor=&MakeLinearColorFromVectorAndFloat(
     "Vector (RGB)"=&Multiply_Vector(
      A=&Multiply_Vector(
       A=&MakeVectorFromLinearColorRGB(Color=$Particles.Initial.FlareCardColor)
       B=&MakeVectorFromLinearColorRGB(Color=&ColorFromCurve(ColorCurve=DataInterface() {propertyValues {
           AlphaCurve {keys [<at 0 1 RCIM_Cubic>]}
           BlueCurve {keys [<at 0 0.8203>,<at 0.4463 0>,<at 1 0.073>]}
           GreenCurve {keys [<at 0 9>,<at 0.4463 4>,<at 1 0.1693>]}
           RedCurve {keys [<at 0 5.0596>,<at 0.4463 1.2383>,<at 1 0>]}
          }}))
      )
      B=&VectorFromFloat(Value=&FN_FrontEndFloatModifier(Gameplay=1, Front End=0.7))
     )
     "Float (Alpha)"=&FloatFromCurve(
      FloatCurve=DataInterface() {propertyValues {Curve {keys [
          {arriveTangent=-1.9727, interpMode=RCIM_Cubic, time=0.1, value=1}
          {arriveTangent=-2.4337, interpMode=RCIM_Cubic, leaveTangent=-2.4337, tangentMode=RCTM_User, time=1, value=0}
         ]}}}
      Scale Curve=&MakeCustomFloatFromBool(Bool=$User.bIsFrontEndPreview, True Float=0.5, False Float=1)
     )
    )
   ) {isSetParameters=true}
   /SetVariables(
    Particles.Update Flash Star Color:LinearColor=&ColorFromCurve(
     ColorCurve=DataInterface() {propertyValues {
       AlphaCurve {keys [<at 0.7378 1>,<at 1 0>]}
       BlueCurve {keys [<at 0.257 0.5>,<at 0.5238 0.3809>,<at 0.9466 0.2366>]}
       GreenCurve {keys [<at 0.257 12>,<at 0.5238 2>,<at 0.9466 1.0039>]}
       RedCurve {keys [<at 0.257 8.3801>,<at 0.5238 1.0849>,<at 0.9466 0.3461>]}
      }}
     CurveIndex=$Particles.Update Flash Star Age
    )
    Particles.Update Flash Star Sprize Size:Vector2f=&Multiply_Vector2D_ByFloat(
     Float=&FloatFromCurve(
      FloatCurve=DataInterface() {propertyValues {Curve {
         defaultValue=1
         keys [
          {arriveTangent=4.1907, interpMode=RCIM_Cubic, leaveTangent=4.1907, tangentMode=RCTM_User, time=0, value=0.5}
          {arriveTangentWeight=0.3485, interpMode=RCIM_Cubic, tangentMode=RCTM_User, time=0.25, value=1.2}
          <at 1 1 RCIM_Cubic>
         ]
        }}}
      CurveIndex=$Particles.Update Flash Star Age
     )
     Vector2D=$Particles.Initial.Flash Star Sprite Size
    )
   ) {isSetParameters=true}
   /SetVariables(Particles.Update Flare Card Size:Vector2f=&Multiply_Vector2D_ByFloat(Float=0.8, Vector2D=$Particles.SpriteSize)) {isSetParameters=true}
   /SetVariables(Particles.ColorDarkGlow:LinearColor=&Multiply_LinearColorByFloat(
     Linear Color=&MultiplyLinearColor(
      Linear Color=&ColorFromCurve(ColorCurve=DataInterface() {propertyValues {
         AlphaCurve {keys [<at 0 1>,<at 1 0>]}
         BlueCurve {keys [<at 0 0>,<at 0.8496 0.0245>]}
         GreenCurve {keys [<at 0 0.3026>,<at 0.8496 0.0625>]}
         RedCurve {keys [<at 0 0.0286>,<at 0.8496 0>]}
        }})
      Scale Factor=&MakeVector4_FromLinearColor(LinearColor=$System.TOD)
     )
     Float=&FN_FrontEndFloatModifier(Gameplay=1, Front End=1)
    )) {isSetParameters=true}
   &Drag(
    Drag=&MaskFloatBySpawnGroup(Masked Float=4, Passthrough Float=2, Spawn Group Mask=4)
    Use Linear Drag=true
    Use Rotational Drag=true
   )
   &SolveForcesAndVelocity(Clamp Velocity=false, Limit Acceleration=false)
   &FN_Light_Attributes(
    Color=&MultiplyLinearColor(Linear Color=$System.TOD, Scale Factor=&MakeVector4_FromLinearColor(LinearColor=<rgba 0.7605 1 0.2497 1>))
    Create Color Attribute=true
    Create Exponent Attribute=true
    Create Position Attribute=true
    Create Radius Attribute=true
    Exponent=3
    Position=<0 5 0>
    Radius=90
    Scale Color=&FloatFromCurve(
     FloatCurve=DataInterface() {propertyValues {Curve {keys [
         {interpMode=RCIM_Cubic, tangentMode=RCTM_User, time=0, value=1}
         {arriveTangentWeight=0.3034, interpMode=RCIM_Cubic, tangentMode=RCTM_User, time=1, value=0}
        ]}}}
     Scale Curve=20
    )
   )
  ]
  rendererClasses [&NiagaraLightRendererProperties, &NiagaraSpriteRendererProperties]
  renderers [
   {rendererClass=&NiagaraLightRendererProperties, rendererIndex=0}
   {rendererClass=&NiagaraSpriteRendererProperties, rendererIndex=1}
   {rendererClass=&NiagaraSpriteRendererProperties, rendererIndex=2}
   {rendererClass=&NiagaraSpriteRendererProperties, rendererIndex=3}
   {rendererClass=&NiagaraSpriteRendererProperties, rendererIndex=4}
  ]
 }
 FN_PhysMatDebris {
  emitterUpdateScript [
   &EmitterState()
   &SpawnBurst_Instantaneous(
    Spawn Count=&RandomRangeInt(Minimum=5, Maximum=10)
    Use Spawn Probability=false
    UseLoopCountLimit=false
   )
  ]
  particleSpawnScript [
   /SetVariables(Particles.BiasedRandom=&FloatFromCurve(
     FloatCurve=DataInterface() {propertyValues {Curve {keys [
         <at 0.2 1 RCIM_Cubic>
         {arriveTangent=-3.3394, interpMode=RCIM_Cubic, leaveTangent=-3.3394, tangentMode=RCTM_User, time=0.5, value=0.5}
         <at 0.8 0 RCIM_Cubic>
        ]}}}
     CurveIndex=&RandomRangeFloat(Minimum=0, Maximum=1)
    )) {isSetParameters=true}
   &InitializeParticle(
    Color Mode=::NewEnumerator0
    Lifetime=&RandomRangeFloat(Minimum=0.5, Maximum=1.25)
    Mass=&Lerp_Float(A=0.6, B=1.2, Alpha=$Particles.BiasedRandom)
    Mass Mode=::NewEnumerator1
    Mesh Scale=&Multiply_VectorByFloat(
     Vector=&RandomRangeVector(Minimum=<0.9 0.4 0.5>, Maximum=<1 1 1>)
     Float=&Lerp_Float(A=0.7, B=1, Alpha=$Particles.BiasedRandom)
    )
    Mesh Scale Mode=::NewEnumerator4
    UsePositionOffset=false
    Write Material Random=true
   )
   &FN_PhysicalSurface(
    Dirt Mesh Count=2
    Glass Mesh Count=3
    Ice Mesh Count=2
    Metal Gravity=&RandomRangeFloat(Minimum=-350, Maximum=-1800)
    Metal Mesh Count=2
    Metal Velocity=&RandomRangeFloat(Minimum=80, Maximum=375)
    Metal Velocity Cone Angle=130
    Physical Surface=$User.PhysicalSurface
    Stone Gravity=&RandomRangeFloat(Minimum=-700, Maximum=-1250)
    Stone Mesh Count=2
    Stone Velocity=&RandomRangeFloat(Minimum=80, Maximum=550)
    Stone Velocity Cone Angle=120
    Wood Gravity=&RandomRangeFloat(Minimum=-450, Maximum=-900)
    Wood Mesh Count=2
    Wood Scale=1.5
    Wood Velocity=&RandomRangeFloat(Minimum=80, Maximum=600)
    Wood Velocity Cone Angle=130
   )
   &AddVelocityInCone(
    Cone Angle=$Output.FN_PhysicalSurface.SurfaceResponseConeAngle
    Module.Use Velocity Falloff On Cone Axis=false
    Velocity Strength=$Output.FN_PhysicalSurface.SurfaceResponseVelocity
   )
   &InitialMeshOrientation(Use Rotation Vector=false)
   &AddRotationalVelocity(Rotation Rate=&RandomVector(Vector Scale=&RandomRangeFloat(Minimum=1, Maximum=2)))
   &ApplyInitialForces()
   /SetVariables(Particles.GravityOverrideValue=$Output.FN_PhysicalSurface.GravityOverride) {isSetParameters=true}
  ]
  particleUpdateScript [
   &ParticleState()
   &GravityForce(Gravity=$Particles.GravityOverrideValue)
   &Collision(
    Advanced Aging Rate=1
    Age Colliding Particles=false
    CPU Collision Trace Channel=ECC_WorldStatic
    Collision Enabled=&Boolean_AndOperation(
     Bool1=&SetBoolByFloatComparison(A=$Particles.Initial.Lifetime, B=1)
     Bool2=&SetBoolByFloatComparison(A=$Particles.Age, B=0.12)
    )
    Control Roll On Collision=true
    Friction=1
    Friction During a Bounce=0.25
    Max CPU Trace Length=100
    Module.Control Rest State Via Surface Normal=false
    Module.EnableMaxCollisionCount=false
    Module.Randomize Collision Normal Vector=true
    Randomize Collision Normal=0.25
    ScaleRateOfRotation=false
   ) {enabled=false}
   &SolveForcesAndVelocity(Clamp Velocity=false, Limit Acceleration=false)
   &ScaleMeshSize(Scale Factor=&VectorFromFloat(Value=&FloatFromCurve(FloatCurve=DataInterface() {propertyValues {Curve {keys [<at 0.6 1 RCIM_Cubic>, {arriveTangent=-6.308, interpMode=RCIM_Cubic, leaveTangent=-6.308, tangentMode=RCTM_User, time=1, value=0}]}}})))
   &DynamicMaterialParameters(Module.Param0WriteEnabled=true, Module.Param1WriteEnabled=true, Module.Param2WriteEnabled=false, Module.Param3WriteEnabled=false)
  ]
  rendererClasses [&NiagaraSpriteRendererProperties, &NiagaraMeshRendererProperties]
  renderers [
   {rendererClass=&NiagaraSpriteRendererProperties}
   {rendererClass=&NiagaraMeshRendererProperties, rendererIndex=1}
   {rendererClass=&NiagaraMeshRendererProperties, rendererIndex=2}
   {rendererClass=&NiagaraMeshRendererProperties, rendererIndex=3}
   {rendererClass=&NiagaraMeshRendererProperties, rendererIndex=4}
   {rendererClass=&NiagaraMeshRendererProperties, rendererIndex=5}
   {rendererClass=&NiagaraMeshRendererProperties, rendererIndex=6}
   {rendererClass=&NiagaraMeshRendererProperties, rendererIndex=7}
   {rendererClass=&NiagaraMeshRendererProperties, rendererIndex=8}
  ]
 }
 Linger_Smoke {
  emitterUpdateScript [
   &EmitterState()
   &SpawnBurst_Instantaneous(Spawn Count=2, Use Spawn Probability=false, UseLoopCountLimit=false)
  ]
  particleSpawnScript [
   &InitializeParticle(
    Color=<rgba 0.1275 0.0903 0.3333 0.5>
    Lifetime=&RandomRangeFloat(Minimum=0.3, Maximum=0.5)
    Sprite Rotation Mode=::NewEnumerator2
    Sprite Size Max=<50 50>
    Sprite Size Min=<35 35>
    Sprite Size Mode=::NewEnumerator2
    UsePositionOffset=false
    Write Material Random=true
   )
   &ShapeLocation(MaskSpawnGroup=false, Override Seed=false, Sphere Radius=12, UseOwnerScale=true)
   &AddVelocity(
    Module.Use Falloff Distance=false
    Override Seed=false
    Velocity Mode=::NewEnumerator1
    Velocity Speed=&RandomRangeFloat(Minimum=5, Maximum=12)
   )
   &CameraOffset(Camera Offset Amount=&RandomRangeFloat(Minimum=15, Maximum=20))
   &DynamicMaterialParameters(
    Index 0 Param 1=3
    Index 0 Param 2=0.5
    Index 0 Param 3=1
    Module.Index0Float0Write=true
    Module.Index0Float1Write=true
    Module.Index0Float2Write=true
    Module.Index0Float3Write=true
   )
  ]
  particleUpdateScript [
   &SubUVAnimation(
    End Frame Range Override=64
    "Sprite Renderer "=DataInterface() {propertyValues {SpriteRenderer=&self:Linger_Smoke_0.Renderer}}
    Start Frame Range Override=45
    UseEndFrame=true
    UseStartFrame=true
   )
   &ParticleState()
   &SpriteRotationRate(Rotation Rate=&RandomRangeFloat(Minimum=50, Maximum=-50))
   &ScaleSpriteSize(
    Uniform Curve Scale=$System.Global Scale
    Uniform Curve Sprite Scale=DataInterface() {propertyValues {Curve {
       defaultValue=0
       keys [
        {arriveTangent=29.5588, interpMode=RCIM_Cubic, leaveTangent=29.5588, tangentMode=RCTM_User, time=0, value=0}
        {
         arriveTangent=2.2374
         arriveTangentWeight=0.1429
         interpMode=RCIM_Cubic
         leaveTangent=2.2374
         leaveTangentWeight=1.2859
         tangentMode=RCTM_User
         time=0.0918
         value=1.8474
        }
        <at 1 2.5>
       ]
      }}}
   )
   &Drag(Drag=3, Use Linear Drag=true, Use Rotational Drag=true)
   &GravityForce(Gravity=<0 0 -100>)
   &SolveForcesAndVelocity(Clamp Velocity=false, Limit Acceleration=false)
   &ScaleColor(
    Scale Alpha=&FloatFromCurve(
     FloatCurve=DataInterface() {propertyValues {Curve {
        defaultValue=1
        keys [
         {arriveTangent=9.2168, interpMode=RCIM_Cubic, leaveTangent=9.2168, tangentMode=RCTM_User, time=0, value=0}
         {arriveTangentWeight=0.1167, interpMode=RCIM_Cubic, tangentMode=RCTM_User, time=0.35, value=1}
         <at 1 0 RCIM_Cubic>
        ]
       }}}
     Scale Curve=&FN_FrontEndFloatModifier(Gameplay=1, Front End=0.8)
    )
    Scale RGB=&MakeVectorFromLinearColorRGB(Color=$System.TOD)
    ScaleA=true
    ScaleRGB=true
   )
  ]
  rendererClasses [&NiagaraSpriteRendererProperties]
  renderers [{rendererClass=&NiagaraSpriteRendererProperties}]
 }
 Smoke_Poof_Up {
  bEnabled=false
  emitterUpdateScript [
   &EmitterState()
   &SpawnBurst_Instantaneous(Spawn Count=4, Use Spawn Probability=false)
  ]
  particleSpawnScript [
   /SetVariables(Particles.VelCurveRand:NiagaraFloat=&FloatFromCurve(
     FloatCurve=DataInterface() {propertyValues {Curve {keys [
         {interpMode=RCIM_Cubic, leaveTangent=0.1132, tangentMode=RCTM_Break, time=0, value=0}
         {arriveTangent=1.9858, interpMode=RCIM_Cubic, leaveTangent=1.7203, tangentMode=RCTM_Break, time=0.3, value=0.4}
         {interpMode=RCIM_Cubic, tangentMode=RCTM_Break, time=1, value=1}
        ]}}}
     CurveIndex=&RandomRangeFloat(Minimum=0.9, Maximum=1)
     Scale Curve=5
    )) {isSetParameters=true}
   /SetVariables(Particles.InitSpriteRotRate=&RandomRangeFloat(Minimum=-1, Maximum=1)) {isSetParameters=true}
   &InitializeParticle(
    Color=&SelectLinearColorFromArray(
     Array Sampling Mode=::NewEnumerator0
     Direct Array Index=?
     Color Selection Array=DataInterface() {propertyValues {ColorData [<rgba 0.1014 0.3672 0.2436 1>, <rgba 0.1079 0.3281 0.1715 1>, <rgba 0.0775 0.2656 0.1371 1>]}}
     Random Seed=0
    )
    Lifetime Max=0.65
    Lifetime Min=0.45
    Lifetime Mode=::NewEnumerator1
    Position Offset Coordinate Space=::Local
    Sprite Rotation Angle=&RandomRangeFloat(Minimum=0, Maximum=360)
    Sprite Rotation Mode=::NewEnumerator0
    Sprite Size Mode=::NewEnumerator1
    Uniform Sprite Size Max=95
    Uniform Sprite Size Min=75
    UsePositionOffset=true
    Write Material Random=true
   )
   /SetVariables(Particles.SpriteSize=&Multiply_Vector2D_ByFloat(Float=1, Vector2D=$Particles.SpriteSize)) {isSetParameters=true}
   &CameraOffset(Camera Offset Amount=&UniformRangedFloat(Minimum=25, Maximum=50))
   &ShapeLocation(
    MaskSpawnGroup=false
    Offset Mode=::NewEnumerator0
    Override Seed=false
    Rotation Coordinate Space=::World
    Sphere Radius=&Multiply_Float(A=15, B=$System.Global Scale)
    UseOwnerScale=true
   )
   &AddVelocityFromPoint(
    Module.Use Falloff Distance=false
    Velocity Strength=&Multiply_Float(A=&Lerp_Float(A=10, B=15, Alpha=$Particles.VelCurveRand), B=$System.Global Scale)
   )
  ]
  particleUpdateScript [
   &ParticleState()
   &Drag(Drag=2, Use Linear Drag=true, Use Rotational Drag=false)
   &SpriteRotationRate(Rotation Rate=&Multiply_Float(
     A=$Particles.InitSpriteRotRate
     B=&FloatFromCurve(
      FloatCurve=DataInterface() {propertyValues {Curve {keys [
          {arriveTangent=-4.4624, interpMode=RCIM_Cubic, leaveTangent=-4.4624, tangentMode=RCTM_User, time=0, value=1}
          {arriveTangentWeight=0.1718, interpMode=RCIM_Cubic, tangentMode=RCTM_User, time=0.5, value=0.15}
         ]}}}
      Scale Curve=1000
     )
    ))
   &ScaleColor(
    Scale Alpha=&FloatFromCurve(
     FloatCurve=DataInterface() {propertyValues {Curve {keys [
         {arriveTangent=0.4602, interpMode=RCIM_Cubic, leaveTangent=0.4602, tangentMode=RCTM_User, time=0, value=0}
         {
          arriveTangent=-0.4587
          arriveTangentWeight=0.1274
          interpMode=RCIM_Cubic
          leaveTangent=-0.4587
          leaveTangentWeight=0.2616
          tangentMode=RCTM_User
          time=0.2
          value=0.7173
         }
         {arriveTangentWeight=0.3988, interpMode=RCIM_Cubic, tangentMode=RCTM_User, time=1, value=0}
        ]}}}
     Scale Curve=&Multiply_Float(
      A=&FloatFromCurve(
       FloatCurve=DataInterface() {propertyValues {Curve {keys [<at 0.2 1 RCIM_Cubic>,<at 0.5 0.9 RCIM_Cubic>,<at 0.85 1 RCIM_Cubic>]}}}
       CurveIndex=Linked() {linkedVariable {name=NPC.FortniteNPC.TimeOfDay_0-1, type=NiagaraFloat}}
      )
      B=&FN_FrontEndFloatModifier(Gameplay=1, Front End=0.75)
     )
    )
    Scale RGB=&Multiply_VectorByFloat(
     Vector=&Multiply_Vector(
      A=&MakeVectorFromLinearColorRGB(Color=$System.TOD)
      B=&MakeVectorFromLinearColorRGB(Color=&ColorFromCurve(ColorCurve=DataInterface() {propertyValues {
          AlphaCurve {keys [<at 0 1 RCIM_Cubic>]}
          BlueCurve {keys [<at 0.0282 2 RCIM_Cubic>,<at 0.1078 1 RCIM_Cubic>]}
          GreenCurve {keys [<at 0.0282 2 RCIM_Cubic>,<at 0.1078 1 RCIM_Cubic>]}
          RedCurve {keys [<at 0.0282 2 RCIM_Cubic>,<at 0.1078 1 RCIM_Cubic>]}
         }}))
     )
     Float=&FloatFromCurve(
      FloatCurve=DataInterface() {propertyValues {Curve {keys [
          {arriveTangent=0.6788, interpMode=RCIM_Cubic, time=0.1, value=1.2}
          {arriveTangent=-1.2539, interpMode=RCIM_Cubic, leaveTangent=-1.2539, time=0.295, value=1}
          <at 0.4988 0.7 RCIM_Cubic>
          {arriveTangent=0.9975, interpMode=RCIM_Cubic, leaveTangent=0.9975, time=0.8, value=1}
          <at 1 1.2 RCIM_Cubic>
         ]}}}
      CurveIndex=Linked() {linkedVariable {name=NPC.FortniteNPC.TimeOfDay_0-1, type=NiagaraFloat}}
      Scale Curve=1.1
     )
    )
    ScaleA=true
    ScaleRGB=true
   )
   &ScaleSpriteSize(Scale Factor=&Vector2DFromFloat(Value=&FloatFromCurve(
      FloatCurve=DataInterface() {propertyValues {Curve {keys [
          {arriveTangent=3.6523, interpMode=RCIM_Cubic, leaveTangent=3.6523, tangentMode=RCTM_User, time=0, value=0.3}
          {arriveTangent=0.3473, interpMode=RCIM_Cubic, leaveTangent=0.3473, tangentMode=RCTM_User, time=0.3, value=0.85}
          {arriveTangent=0.1688, arriveTangentWeight=0.3375, interpMode=RCIM_Cubic, leaveTangent=0.1688, tangentMode=RCTM_User, time=1.1905, value=1}
         ]}}}
      Scale Curve=$System.Global Scale
     )))
   &DynamicMaterialParameters(
    Index 0 Param 1=&Add_Float(
     A=&RandomRangeFloat(Minimum=0, Maximum=1)
     B=&FloatFromCurve(
      FloatCurve=DataInterface() {propertyValues {Curve {keys [
          {arriveTangent=2.7533, interpMode=RCIM_Cubic, leaveTangent=1.8958, tangentMode=RCTM_User, time=0, value=1}
          {arriveTangent=0.8292, interpMode=RCIM_Cubic, leaveTangent=0.8292, tangentMode=RCTM_User, time=0.321, value=1.3453}
          {arriveTangent=0.0529, interpMode=RCIM_Cubic, leaveTangent=0.2021, tangentMode=RCTM_User, time=1, value=1.5}
         ]}}}
      CurveIndex=&Add_Float(A=&Multiply_Float(A=$Particles.Initial.MaterialRandom, B=0.5), B=$Particles.NormalizedAge)
     )
    )
    Index 0 Param 2=&FloatFromCurve(FloatCurve=DataInterface() {propertyValues {Curve {keys [
         {arriveTangent=0.0453, interpMode=RCIM_Cubic, tangentMode=RCTM_Break, time=0, value=-1}
         {arriveTangent=0.0604, interpMode=RCIM_Cubic, leaveTangent=0.0604, tangentMode=RCTM_User, time=1, value=-1}
        ]}}})
    Index 0 Param 3=&FloatFromCurve(
     FloatCurve=DataInterface() {propertyValues {Curve {keys [
         {arriveTangent=0.0632, interpMode=RCIM_Cubic, leaveTangent=0.0632, tangentMode=RCTM_User, time=0, value=1}
         {
          arriveTangent=-1.6858
          arriveTangentWeight=0.0333
          interpMode=RCIM_Cubic
          leaveTangent=-1.6858
          leaveTangentWeight=0.3137
          tangentMode=RCTM_User
          time=0.05
          value=0.1
         }
         {arriveTangentWeight=0.1027, interpMode=RCIM_Cubic, tangentMode=RCTM_Break, time=0.2, value=0}
        ]}}}
     Scale Curve=&RandomRangeFloat(Minimum=0.25, Maximum=0.5)
    )
    Index 0 Param 4=&FloatFromCurve(FloatCurve=DataInterface() {propertyValues {Curve {keys [<at 0.5 1>,<at 1 0>]}}})
    Index 1 Param 1=&FloatFromCurve(
     FloatCurve=DataInterface() {propertyValues {Curve {keys [
         {interpMode=RCIM_Cubic, tangentMode=RCTM_Break, time=0, value=1}
         {arriveTangentWeight=0.0833, interpMode=RCIM_Cubic, tangentMode=RCTM_Break, time=0.25, value=0.3}
        ]}}}
     Scale Curve=&RandomRangeFloat(Minimum=1.5, Maximum=3)
    )
    Index 1 Param 2=&FloatFromCurve(
     FloatCurve=DataInterface() {propertyValues {Curve {keys [
         {interpMode=RCIM_Cubic, leaveTangent=-2.5489, tangentMode=RCTM_Break, time=0, value=1}
         {interpMode=RCIM_Cubic, leaveTangent=0.9066, tangentMode=RCTM_Break, time=1, value=0}
        ]}}}
     Scale Curve=&RandomRangeFloat(Minimum=0.01, Maximum=0.1)
    )
    Index 1 Param 3=&FloatFromCurve(FloatCurve=DataInterface() {propertyValues {Curve {keys [
         {interpMode=RCIM_Cubic, leaveTangent=6.3917, tangentMode=RCTM_Break, time=0, value=0.2}
         {arriveTangent=1.7549, interpMode=RCIM_Cubic, leaveTangent=10.6767, tangentMode=RCTM_User, time=0.25, value=1}
        ]}}})
    Index 1 Param 4=&FloatFromCurve(
     FloatCurve=DataInterface() {propertyValues {Curve {keys [{arriveTangent=-0.136, interpMode=RCIM_Cubic, leaveTangent=-0.136, tangentMode=RCTM_User, time=0, value=1}, <at 1 0.95>]}}}
     Scale Curve=&RandomRangeFloat(Minimum=1.3, Maximum=2)
    )
    Index 2 Param 1=&FloatFromCurve(FloatCurve=DataInterface() {propertyValues {Curve {keys [
         {interpMode=RCIM_Cubic, leaveTangent=4.2485, tangentMode=RCTM_Break, time=0, value=0}
         {arriveTangent=1.0548, interpMode=RCIM_Cubic, leaveTangent=1.0548, tangentMode=RCTM_User, time=0.17, value=0.5715}
         <at 1 1>
        ]}}})
    Index 2 Param 2=&FloatFromCurve(
     FloatCurve=DataInterface() {propertyValues {Curve {keys [
         {arriveTangent=1.9056, interpMode=RCIM_Cubic, leaveTangent=1.9056, tangentMode=RCTM_User, time=0, value=0.2}
         {arriveTangent=0.0259, interpMode=RCIM_Cubic, leaveTangent=0.0259, tangentMode=RCTM_User, time=1, value=1.4}
        ]}}}
     Scale Curve=&RandomRangeFloat(Minimum=0.1, Maximum=0.2)
    )
    Index 2 Param 3=&FloatFromCurve(
     FloatCurve=DataInterface() {propertyValues {Curve {keys [
         {interpMode=RCIM_Cubic, leaveTangent=0.0393, tangentMode=RCTM_Break, time=0, value=1}
         {interpMode=RCIM_Cubic, leaveTangent=11.1233, tangentMode=RCTM_Break, time=1, value=1.5}
        ]}}}
     Scale Curve=0.7
    )
    Index 2 Param 4=&FloatFromCurve(
     FloatCurve=DataInterface() {propertyValues {Curve {keys [
         {arriveTangent=-0.6867, interpMode=RCIM_Cubic, leaveTangent=-0.6867, tangentMode=RCTM_User, time=0, value=1}
         {interpMode=RCIM_Cubic, leaveTangent=-1.2956, tangentMode=RCTM_Break, time=1, value=0.75}
        ]}}}
     Scale Curve=1.5
    )
    Module.Index0Float0Write=true
    Module.Index0Float1Write=true
    Module.Index0Float2Write=true
    Module.Index0Float3Write=true
    Module.Index1Float0Write=true
    Module.Index1Float1Write=true
    Module.Index1Float2Write=true
    Module.Index1Float3Write=true
    Module.Index2Float0Write=true
    Module.Index2Float1Write=true
    Module.Index2Float2Write=true
    Module.Index2Float3Write=true
    Write Parameter Index 1=::NewEnumerator1
    Write Parameter Index 2=::NewEnumerator1
   )
   &GravityForce(Gravity=&VectorFromCurve(VectorCurve=DataInterface() {propertyValues {
       XCurve {keys [<at 0 0>,<at 1 0>]}
       YCurve {keys [<at 0 0>,<at 1 0>]}
       ZCurve {keys [
         <at 0 0 RCIM_Cubic>
         {arriveTangent=-924.4257, interpMode=RCIM_Cubic, leaveTangent=-924.4257, tangentMode=RCTM_User, time=0.3387, value=-100}
         {arriveTangent=-1826.4429, interpMode=RCIM_Cubic, leaveTangent=-1826.4429, tangentMode=RCTM_User, time=0.498, value=-150}
        ]}
      }}))
   &SolveForcesAndVelocity(Clamp Velocity=false, Limit Acceleration=false)
   /SetVariables(Particles.UnlitSmokeColor:LinearColor=&MultiplyLinearColor(Linear Color=$Particles.Color, Scale Factor=<1.5 1.5 1.5 0.65>)) {isSetParameters=true}
  ]
  rendererClasses [&NiagaraSpriteRendererProperties]
  renderers [
   {rendererClass=&NiagaraSpriteRendererProperties}
   {rendererClass=&NiagaraSpriteRendererProperties, rendererIndex=1}
  ]
 }
 Sparks_Combined {
  emitterUpdateScript [
   &EmitterState(
    Enable Distance Culling=true
    Max Distance Response=::KillImmediately
    Module.bApplySpawnCountScaleByDistance=true
    Module.bUseMaxDistance=true
    Module.bUseMinDistance=false
    Scalability Mode=::NewEnumerator1
    Scale Spawn Count=true
    Spawn Count Scale By Distance=DataInterface() {propertyValues {Curve {keys [<at 0.3676 0.9657>,<at 0.9962 0.4982>]}}}
   )
   &SpawnBurst_Instantaneous(Spawn Count=&UniformRangedInt(Minimum=30, Maximum=35), Use Spawn Probability=true) {module=SpawnBurst_Instantaneous002}
  ]
  particleSpawnScript [
   &ShapeLocation(
    MaskSpawnGroup=false
    Override Seed=false
    Rotation Axis=$System.DirectionVector
    Rotation Mode=::NewEnumerator1
    Sphere Radius=10
    UseOwnerScale=true
   )
   /SetVariables(Particles.SlidingScale:NiagaraFloat=&FloatFromCurve(
     FloatCurve=DataInterface() {propertyValues {Curve {keys [<at 0 1>,<at 0.25 0.2>,<at 1 0>]}}}
     CurveIndex=&RandomRangeFloat(Minimum=0, Maximum=1)
     Scale Curve=$System.Global Scale
    )) {isSetParameters=true}
   &InitializeParticle(
    Color=&ColorFromCurve(
     ColorCurve=DataInterface() {propertyValues {
       AlphaCurve {keys [<at 0 1 RCIM_Cubic>]}
       BlueCurve {keys [
         <at 0.0912 4>
         {arriveTangent=-20.1663, interpMode=RCIM_Cubic, leaveTangent=-20.1663, time=0.1728, value=4}
         <at 0.2553 0.6901>
         <at 0.4766 0.5679>
         <at 0.5364 0 RCIM_Cubic>
        ]}
       GreenCurve {keys [
         <at 0.0912 0.2227>
         {arriveTangent=4.3852, interpMode=RCIM_Cubic, leaveTangent=4.3852, time=0.1728, value=0}
         <at 0.2553 0.9424>
         <at 0.4766 0.8244>
         <at 0.5364 1 RCIM_Cubic>
        ]}
       RedCurve {keys [
         <at 0.0912 0>
         {arriveTangent=7.5528, interpMode=RCIM_Cubic, leaveTangent=7.5528, time=0.1728, value=0.6121}
         <at 0.2553 1.2396>
         <at 0.4766 0.7758>
         <at 0.5364 0.3096 RCIM_Cubic>
        ]}
      }}
     CurveIndex=&RandomRangeFloat(Minimum=0, Maximum=1)
    )
    Lifetime=&FloatFromCurve(
     FloatCurve=DataInterface() {propertyValues {Curve {keys [
         {arriveTangent=-1.1082, interpMode=RCIM_Cubic, leaveTangent=-1.1082, tangentMode=RCTM_User, time=0, value=0.75}
         {arriveTangent=-0.2452, tangentMode=RCTM_User, time=1, value=0.4}
        ]}}}
     CurveIndex=&RandomRangeFloat(Minimum=0, Maximum=1)
     Scale Curve=&Multiply_Float(A=$System.Global Scale, B=&Lerp_Float(A=0.5, B=1, Alpha=$Particles.SlidingScale))
    )
    Position Offset:NiagaraPosition=$Particles.ShapeLocation.ShapePosition
    Sprite Size Mode=::NewEnumerator3
    Uniform Sprite Size=&FloatFromCurve(
     FloatCurve=DataInterface() {propertyValues {Curve {keys [<at 0 10>,<at 0.25 5>,<at 1 4>]}}}
     CurveIndex=$Particles.SlidingScale
     Scale Curve=$System.Global Scale
    )
    UsePositionOffset=true
    Write Material Random=true
   )
   &CameraOffset(Camera Offset Amount=15)
   &AddVelocityFromPoint(
    Module.Use Falloff Distance=false
    Origin Offset Coordinate Space=::World
    Velocity Strength=&Multiply_Float(A=&RandomRangeFloat(Minimum=450, Maximum=150), B=$System.Global Scale)
   )
   &DynamicMaterialParameters(
    Index 0 Param 1=2
    Index 0 Param 2=20
    Index 0 Param 3=1
    Module.Index0Float0Write=true
    Module.Index0Float1Write=true
    Module.Index0Float2Write=true
    Module.Index0Float3Write=true
   )
  ]
  particleUpdateScript [
   &ParticleState()
   &AccelerationForce(Acceleration=&Multiply_VectorByFloat(Vector=&RandomRangeVector(Minimum=<0 0 -1000>, Maximum=<0 0 -500>), Float=1))
   &Drag(Drag=&RandomRangeFloat(Minimum=2, Maximum=6), Use Linear Drag=true, Use Rotational Drag=true)
   &SolveForcesAndVelocity(Clamp Velocity=false, Limit Acceleration=false)
   &ScaleSpriteSizeBySpeed(Max Scale Factor=<0.6 3>, Min Scale Factor=<0.2 0.4>, Velocity Threshold=600)
   &ScaleColor(
    Scale Alpha=&FloatFromCurve(
     FloatCurve=DataInterface() {propertyValues {Curve {keys [
         {interpMode=RCIM_Cubic, tangentMode=RCTM_User, time=0, value=1}
         {arriveTangent=-0.3845, interpMode=RCIM_Cubic, leaveTangent=-0.3845, tangentMode=RCTM_User, time=0.5, value=0.9}
         {arriveTangent=-0.2828, interpMode=RCIM_Cubic, leaveTangent=-0.2828, tangentMode=RCTM_User, time=1, value=0}
        ]}}}
     Scale Curve=&RandomRangeFloat(Minimum=0.25, Maximum=5)
    )
    Scale RGB=&MakeVectorFromLinearColorRGB(Color=&ColorFromCurve(ColorCurve=DataInterface() {propertyValues {
        AlphaCurve {keys [<at 0.333 1>,<at 1 0>]}
        BlueCurve {keys [<at 0 0>,<at 0.1605 0.0383>,<at 0.5 0.7409>,<at 1 0.0951>]}
        GreenCurve {keys [<at 0 6>,<at 0.1605 5>,<at 0.5 1.2841>,<at 1 0.1474>]}
        RedCurve {keys [<at 0 5.2616>,<at 0.1605 4.8868>,<at 0.5 4>,<at 1 1.1415>]}
       }}))
    ScaleA=true
    ScaleRGB=true
   )
   /SetVariables(
    Particles.FlareSparkSize:Vector2f=&Multiply_Vector2D_ByFloat(
     Float=&FloatFromCurve(
      FloatCurve=DataInterface() {propertyValues {Curve {
         defaultValue=1
         keys [
          {interpMode=RCIM_Cubic, tangentMode=RCTM_User, time=0.7, value=1}
          {arriveTangent=-2.7013, interpMode=RCIM_Cubic, leaveTangent=-2.7013, tangentMode=RCTM_User, time=1, value=0.7}
         ]
        }}}
      Scale Curve=2
     )
     Vector2D=$Particles.SpriteSize
    )
    Particles.SparkleSparkSize:Vector2f=&Multiply_Vector2D_ByFloat(
     Float=&Multiply_Float(
      A=&RandomRangeFloat(Minimum=0.6, Maximum=1.2)
      B=&FloatFromCurve(FloatCurve=DataInterface() {propertyValues {Curve {
          defaultValue=1
          keys [
           {interpMode=RCIM_Cubic, tangentMode=RCTM_User, time=0.7, value=1}
           {arriveTangent=-4.3053, interpMode=RCIM_Cubic, leaveTangent=-4.3053, tangentMode=RCTM_User, time=1, value=0.5}
          ]
         }}})
     )
     Vector2D=$Particles.SpriteSize
    )
   ) {isSetParameters=true}
  ]
  rendererClasses [&NiagaraSpriteRendererProperties]
  renderers [{rendererClass=&NiagaraSpriteRendererProperties}]
 }
}
summary {
 systemName=NS_Pickaxe_MicStand_TaperSlink_Impact
 emitters [
  {
   emitterName=FN_PhysMatDebris
   rendererClasses [&NiagaraSpriteRendererProperties, &NiagaraMeshRendererProperties]
  }
  {bEnabled=false, emitterName=Smoke_Poof_Up, rendererClasses [&NiagaraSpriteRendererProperties]}
  {emitterName=Sparks_Combined, rendererClasses [&NiagaraSpriteRendererProperties]}
  {emitterName=Linger_Smoke, rendererClasses [&NiagaraSpriteRendererProperties]}
  {
   emitterName=Core_Impact_Burst001
   rendererClasses [&NiagaraLightRendererProperties, &NiagaraSpriteRendererProperties]
  }
 ]
 userVariables [
  User.bIsFrontEnd {type=NiagaraBool}
  User.bIsFrontEndPreview {type=NiagaraBool}
  User.PhysicalSurface {type=NiagaraInt32}
 ]
}
