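# Physics asset for the TaperSlink mic stand backpack cosmetic, giving its skeleton a jointed chain that can flex slightly.
# Four capsule bodies cover the handle and three bend joints, linked in series from handle through bend_joint_01 and 02 to 03.
# Each constraint locks twist and swing1, leaving only a narrow swing2 range, so each link bends on a single axis, with the handle link allowed slightly more.
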PhysicsAsset /SparksCosmetics/Instrument/MicStand/TaperSlink/Meshes/TaperSlink_MicStand_Backpack_Physics
bodies {
 bend_joint_01 [{shapeType=Capsule, length=18, radius=4, center=<8.5 0 0>, rotation=<rot 90 0 0>}]
 bend_joint_02 [{shapeType=Capsule, length=13, radius=3.8, shapeName=bend_joint_02_capsule, center=<7 0 0>, rotation=<rot 90 0 0>}]
 bend_joint_03 [{shapeType=Capsule, length=20, radius=4, center=<10 0 0>, rotation=<rot 90 0 0>}]
 handle [{shapeType=Capsule, length=10, radius=4}]
}
constraints [
 {
  bone1Name=bend_joint_01
  bone2Name=handle
  swing1LimitDegrees=0
  swing1Motion=Locked
  swing2LimitDegrees=8
  twistLimitDegrees=0
  twistMotion=Locked
 }
 {
  bone1Name=bend_joint_03
  bone2Name=bend_joint_02
  swing1LimitDegrees=0
  swing1Motion=Locked
  swing2LimitDegrees=5
  twistLimitDegrees=0
  twistMotion=Locked
 }
 {
  bone1Name=bend_joint_02
  bone2Name=bend_joint_01
  swing1LimitDegrees=0
  swing1Motion=Locked
  swing2LimitDegrees=5
  twistLimitDegrees=0
  twistMotion=Locked
 }
]
